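# Material function for the Pencil cosmetic that composites an animated, tintable chibi over the base color, outputting color, chibi alpha and a reactivity alpha.
# Chibi_Index picks a 4x4 atlas cell on UV1, with stepped time-based sine offsets giving the head choppy motion.
# An FX atlas masks head, hair and body for parameter tinting, and Chibi_Blink blends in closed eyes.
# Reactivity combines a SkinFX channel, UV masks and parameters, minus the chibi.
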MaterialFunction /BRCosmetics/Characters/Player/Male/Medium/Bodies/M_MED_Pencil_A/PencilMaterials/MFC_Chibi
graph {nodes [
  n1 = TextureCoordinate() {coordinateIndex=1}
  n2 = Constant() {r=32}
  Divide(A=Round(None=Multiply(A=@n1, B=@n2)), B=@n2)
  n3 = NamedRerouteUsage()
  n4 = NamedRerouteUsage()
  FunctionOutput(None=LinearInterpolate(A=NamedRerouteUsage(), B=Clamp(None=@n3) {maxDefault=0.9}, Alpha=@n4)) {outputName=Chibi}
  FunctionOutput(None=Multiply(A=@n4, B=ScalarParameter() {defaultValue=0.8, parameterName=ChibiAlphaIntensity})) {outputName=ChibiAlpha, sortPriority=1}
  n5 = NamedRerouteUsage()
  FunctionOutput(None=Saturate(None=Subtract(A=@n5, B=@n4))) {outputName=ReactivityAlpha, sortPriority=2}
  n6 = NamedRerouteUsage()
  n7 = Multiply(
   A=Multiply(A=Time()) {constB=0.75}
   B=LinearInterpolate(
    B=ScalarParameter() {defaultValue=1, parameterName=Chibi_HeadRotationSpeed}
    Alpha=ScalarParameter() {defaultValue=1, group=Toon_Shading_Pencil, parameterName=PencilBackpack_ReactivityToggle}
   ) {constA=1}
  )
  n8 = ConstantBiasScale(None=Step(Y=Sine(None=@n7)) {constX=0}) {bias=-0.5, scale=1.5}
  n9 = ScalarParameter() {defaultValue=0.05, parameterName=Chibi_HeadRotationAmount}
  n10 = Sine(None=Multiply(A=Time()) {constB=1.5})
  MaterialFunctionCall(
   Rotation Center=Add(A=Divide(A=@n6) {constB=4}, B=Constant2Vector() {g=0.18, r=0.125})
   "Rotation Angle (0-1)"=LinearInterpolate(B=Multiply(A=@n8, B=@n9), Alpha=Step(Y=@n10) {constX=0})
  ) {materialFunction=&CustomRotator}
  n11 = FunctionInput(Preview=TextureSample() {texture=/Game/Accessories/FORT_Backpacks/Backpack_F_MED_PencilCherry/Textures/T_F_MED_PencilCherry_Backpack_D}.RGB) {inputName=BaseColor_IN}
  Multiply(A=@n11, B=Reroute())
  n12 = Multiply(
   A=Multiply(A=Time()) {constB=0.75}
   B=LinearInterpolate(
    B=ScalarParameter() {defaultValue=1, parameterName=Chibi_HeadTranslationSpeed}
    Alpha=ScalarParameter() {defaultValue=1, group=Toon_Shading_Pencil, parameterName=PencilBackpack_ReactivityToggle}
   ) {constA=1}
  )
  n13 = ConstantBiasScale(None=Step(Y=Sine(None=@n12)) {constX=0}) {bias=-0.5, scale=1.5}
  n14 = ScalarParameter() {defaultValue=0.0062, parameterName=Chibi_HeadTranslationAmount}
  NamedRerouteDeclaration(None=MaterialFunctionCall(X=Multiply(A=@n13, B=@n14), Y=Constant()) {materialFunction=&MakeFloat2}.Result) {name=UVTranslation, nodeColor=<rgba 0 0.2969 1 1>}
  n15 = TextureObject() {
   samplerType=SAMPLERTYPE_LinearColor
   texture=/Game/Accessories/FORT_Backpacks/Backpack_F_MED_PencilCherry/Textures/T_F_MED_PencilCherry_Backpack_Chibi_Atlas_FX
  }
  n16 = Constant() {r=-3}
  n17 = TextureCoordinate() {coordinateIndex=1}
  n18 = ScalarParameter() {defaultValue=0.9, parameterName=Chibi_Scale}
  n19 = ScalarParameter() {group=Toon_Shading_Pencil, parameterName=Chibi_Index}
  n20 = Constant() {r=4}
  n21 = AppendVector(A=Fmod(A=@n19, B=@n20), B=Floor(None=Divide(A=@n19, B=@n20) {constB=4}))
  n22 = Divide(
   A=Add(
    A=MaterialFunctionCall(UVs=@n17, Texture Scale=@n18) {materialFunction=&ScaleUVsByCenter}.UVs
    B=NamedRerouteDeclaration(None=@n21) {name=Index2DOffset, nodeColor=<rgba 1 0 0.1641 1>}
   )
   B=@n20
  ) {constB=4}
  n23 = Add(A=@n22, B=NamedRerouteUsage())
  n24 = TextureSample!(UVs=@n23, Tex=@n15, Bias=@n16, Apply View MipBias) {
   mipValueMode=TMVM_MipBias
   texture=/Game/Accessories/FORT_Backpacks/Backpack_F_MED_PencilCherry/Textures/T_F_MED_PencilCherry_Backpack_Chibi_Atlas
  }
  n25 = MaterialFunctionCall(Float3=@n24.RGB) {materialFunction=&BreakOutFloat3Components}
  n26 = NamedRerouteDeclaration(None=@n25.R) {name=HeadMask, nodeColor=<rgba 1 0 0.7734 1>}
  NamedRerouteDeclaration(None=Multiply(A=@n26, B=@n25.B)) {name=HairMask01, nodeColor=<rgba 0 1 0.8125 1>}
  n27 = OneMinus(None=@n26)
  n28 = TextureSample(UVs=@n22, Tex=@n15, Bias=@n16) {
   mipValueMode=TMVM_MipBias
   texture=&T_F_MED_PencilCherry_Backpack_Chibi_Atlas
  }
  n29 = MaterialFunctionCall(Float3=@n28.RGB) {materialFunction=&BreakOutFloat3Components}
  NamedRerouteDeclaration(None=Multiply(A=@n27, B=@n29.G)) {name=BodyColor01, nodeColor=<rgba 0 1 0.1562 1>}
  NamedRerouteDeclaration(None=Multiply(A=@n27, B=@n29.B)) {name=BodyColor02, nodeColor=<rgba 0 1 0.1562 1>}
  n30 = Multiply(A=Time()) {constB=0.75}
  n31 = LinearInterpolate(
   B=ScalarParameter() {defaultValue=1, parameterName=Chibi_HeadBobSpeed}
   Alpha=ScalarParameter() {defaultValue=1, group=Toon_Shading_Pencil, parameterName=PencilBackpack_ReactivityToggle}
  ) {constA=1}
  n32 = Step(Y=Sine(None=Multiply(A=@n30, B=@n31))) {constX=0.5}
  n33 = ScalarParameter() {defaultValue=-0.0078, parameterName=Chibi_HeadBobAmount}
  NamedRerouteDeclaration(None=MaterialFunctionCall(X=Constant(), Y=Multiply(A=@n32, B=@n33)) {materialFunction=&MakeFloat2}.Result) {name=UVBobbing, nodeColor=<rgba 0 1 0.625 1>}
  n34 = LinearInterpolate(A=@n22, B=@n23, Alpha=NamedRerouteUsage())
  n35 = TextureSample(
   UVs=@n34
   Tex=TextureObject() {texture=&T_F_MED_PencilCherry_Backpack_Chibi_Atlas}
   Bias=@n16
  ) {
   desc=Head Bobbing
   mipValueMode=TMVM_MipBias
   texture=&T_F_MED_PencilCherry_Backpack_Chibi_Atlas
  }
  n36 = NamedRerouteDeclaration(None=@n35.A) {name=ChibiAlpha, nodeColor=<rgba 0 0.2734 1 1>}
  n37 = Subtract(A=@n29.R, B=@n26)
  NamedRerouteDeclaration(None=Saturate(None=Subtract(A=@n36, B=@n37))) {name=FinalChibiAlpha, nodeColor=<rgba 0.6875 1 0 1>}
  n38 = Time()
  n39 = Multiply(A=@n38) {constB=0.75}
  n40 = LinearInterpolate(
   B=ScalarParameter() {defaultValue=1, parameterName=Chibi_HeadBobbingSpeed}
   Alpha=ScalarParameter() {defaultValue=1, group=Toon_Shading_Pencil, parameterName=PencilBackpack_ReactivityToggle}
  ) {constA=1}
  n41 = Step(Y=Sine(None=Multiply(A=@n39, B=@n40))) {constX=0}
  n42 = ScalarParameter() {defaultValue=0.0062, parameterName=Chibi_HeadBobAmount_H}
  n43 = Multiply(A=@n38) {constB=0.375}
  n44 = LinearInterpolate(A=Multiply(B=@n42) {constA=-1, constB=-1}, B=@n42, Alpha=Step(Y=Sine(None=@n43)) {constX=0})
  n45 = ScalarParameter() {defaultValue=-0.0078, parameterName=Chibi_HeadBobAmount_V}
  NamedRerouteDeclaration(None=MaterialFunctionCall(X=Multiply(A=@n41, B=@n44), Y=Multiply(A=@n41, B=@n45)) {materialFunction=&MakeFloat2}.Result) {name=UVdiagonalOffset, nodeColor=<rgba 0 1 0.625 1>}
  n46 = ComponentMask(None=TextureCoordinate()) {g=true}
  n47 = VectorParameter() {group=Toon_Shading_Pencil, parameterName=PencilBackpack_IconLocation, defaultValue=<rgba 0.1 0.415 0 1>}
  n48 = Add(
   A=Step(Y=@n46) {constX=0.38}
   B=SphereMask(A=TextureCoordinate(), B=ComponentMask(None=@n47.RGB) {g=true, r=true}, Radius=Constant() {r=0.039}) {attenuationRadius=0.039}
  )
  n49 = TextureSample!(UVs, Tex, Apply View MipBias) {
   samplerType=SAMPLERTYPE_Masks
   texture=/Game/Accessories/FORT_Backpacks/Backpack_F_MED_PencilCherry/Textures/T_F_MED_PencilCherry_Backpack_FX
  }
  n50 = ChannelMaskParameter(None=FunctionInput(Preview=@n49.RGB) {inputName=SkinFX}) {group=Toon_Shading_Pencil, maskChannel=Blue, parameterName=PencilBackpack_IconChannel, defaultValue=<rgba 0 0 1 0>}
  n51 = ScalarParameter() {group=Toon_Shading_Pencil, parameterName=PencilBackpack_Reactivity}
  n52 = ScalarParameter() {defaultValue=1, group=Toon_Shading_Pencil, parameterName=PencilBackpack_ReactivityToggle}
  n53 = NamedRerouteDeclaration(None=Multiply(A=Multiply(A=@n48, B=@n50.R), B=Multiply(A=@n51, B=@n52))) {name=ReactivityAlpha, nodeColor=<rgba 1 0 0.9844 1>}
  NamedRerouteDeclaration(None=Clamp(None=LinearInterpolate(A=@n11, B=@n11, Alpha=@n53)) {maxDefault=0.95}) {name=FinalReactiveColor, nodeColor=<rgba 1 0 0.9375 1>}
  NamedRerouteDeclaration(None=Multiply(A=@n26, B=@n25.G)) {name=HairMask02, nodeColor=<rgba 0 0.4896 0.3978 1>}
  n54 = NamedRerouteDeclaration(None=@n35.RGB) {name=ChibiColor, nodeColor=<rgba 0.7578 1 0 1>}
  n55 = TextureSample!(
   UVs=@n34
   Tex=TextureObject() {texture=/Game/Accessories/FORT_Backpacks/Backpack_F_MED_PencilCherry/Textures/T_F_MED_PencilCherry_Backpack_Chibi_EyesBlink_Atlas}
   Bias=@n16
   Apply View MipBias
  ) {
   mipValueMode=TMVM_MipBias
   texture=&T_F_MED_PencilCherry_Backpack_Chibi_Atlas
  }
  n56 = NamedRerouteDeclaration(None=@n55.RGB) {name=ChibiColor_EyesClosed, nodeColor=<rgba 0.7578 1 0 1>}
  n57 = ScalarParameter() {group=Toon_Shading_Pencil, parameterName=Chibi_Blink}
  NamedRerouteDeclaration(None=LinearInterpolate(B=LinearInterpolate(A=@n54, B=@n56, Alpha=@n57), Alpha=NamedRerouteUsage()) {desc=head bob color with alpha}) {name=CustomColor_IN, nodeColor=<rgba 0.5 0 1 1>}
  n58 = NamedRerouteUsage()
  n59 = VectorParameter() {group=Toon_Shading_Pencil, parameterName=Chibi_HairColor_Default_01, defaultValue=<rgba 0 1 0 1>}
  n60 = VectorParameter() {group=Toon_Shading_Pencil, parameterName=Chibi_HairColor_01}
  n61 = LinearInterpolate(A=@n58, B=Multiply(A=LinearInterpolate(A=@n59.RGB, B=@n60.RGB, Alpha=@n60.RGB), B=@n58), Alpha=NamedRerouteUsage())
  n62 = VectorParameter() {group=Toon_Shading_Pencil, parameterName=Chibi_HairColor_Default_02, defaultValue=<rgba 0 0 1 1>}
  n63 = VectorParameter() {group=Toon_Shading_Pencil, parameterName=Chibi_HairColor_02}
  n64 = LinearInterpolate(A=@n61, B=Multiply(A=@n61, B=LinearInterpolate(A=@n62.RGB, B=@n63.RGB, Alpha=@n63.RGB)), Alpha=NamedRerouteUsage())
  n65 = VectorParameter() {group=Toon_Shading_Pencil, parameterName=Chibi_BodyColor_Default_01, defaultValue=<rgba 0 1 0 1>}
  n66 = VectorParameter() {group=Toon_Shading_Pencil, parameterName=Chibi_BodyColor_01}
  n67 = LinearInterpolate(A=@n64, B=Multiply(A=LinearInterpolate(A=@n65.RGB, B=@n66.RGB, Alpha=@n66.RGB), B=@n64), Alpha=NamedRerouteUsage())
  n68 = VectorParameter() {group=Toon_Shading_Pencil, parameterName=Chibi_BodyColor_02}
  n69 = LinearInterpolate(
   A=VectorParameter() {group=Toon_Shading_Pencil, parameterName=Chibi_BodyColor_Default_02, defaultValue=<rgba 0 0 1 1>}.RGB
   B=@n68.RGB
   Alpha=@n68.RGB
  )
  NamedRerouteDeclaration(None=LinearInterpolate(A=@n67, B=Multiply(A=@n67, B=@n69), Alpha=NamedRerouteUsage())) {name=CustomColor_OUT, nodeColor=<rgba 0.4297 0 1 1>}
  ScalarParameter() {defaultValue=1, group=Toon_Shading_Pencil, parameterName=Chibi_HeadBob}
  ScalarParameter() {defaultValue=1, group=Toon_Shading_Pencil, parameterName=Chibi_HeadBob}
  Step(X=@n46) {constX=0.1, constY=0.0001}
  VectorParameter() {group=Toon_Shading_Pencil, parameterName=PencilBackpack_ReactivityBGColorMulti, defaultValue=<rgba 1 0.8629 0.0288 1>}
 ]}
