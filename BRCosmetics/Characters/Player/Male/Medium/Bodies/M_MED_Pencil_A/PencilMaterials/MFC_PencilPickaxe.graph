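# Material function for the Pencil pickaxe toon shading that adds an animated, noise-broken grid of color-shifted shading to the lit custom colors.
# Two panning noise samples and a slightly rotated grid sample, all from one mask texture projected on pre-skinned XZ position, multiply into a smoothstepped mask.
# That mask blends in a hue-shifted, saturation- and brightness-adjusted color, limited to a selectable channel of the SkinFX input.
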MaterialFunction /BRCosmetics/Characters/Player/Male/Medium/Bodies/M_MED_Pencil_A/PencilMaterials/MFC_PencilPickaxe
graph {nodes [
  n1 = NamedRerouteUsage()
  n2 = ScalarParameter() {defaultValue=-0.07, group=Toon_Shading_Pencil, parameterName=PencilPickaxe_Hueshift}
  n3 = NamedRerouteDeclaration(None=FunctionInput() {inputName=CustomColorsLit}) {name=CustomColorsLit, nodeColor=<rgba 0.6172 0 1 1>}
  n4 = Multiply(
   A=Desaturation(
    None=MaterialFunctionCall(Hue Shift Percentage=@n2, Texture=@n3) {materialFunction=&HueShift}.Result
    Fraction=ScalarParameter() {defaultValue=-1, group=Toon_Shading_Pencil, parameterName=PencilPickaxe_Saturation}
   )
   B=ScalarParameter() {defaultValue=0.7, group=Toon_Shading_Pencil, parameterName=PencilPickaxe_Brightness}
  )
  n5 = ScalarParameter() {defaultValue=0.1, group=Toon_Shading_Pencil, parameterName=PencilPickaxe_GridSmoothstepMin}
  n6 = Time()
  n7 = ComponentMask(None=VectorParameter() {group=Toon_Shading_Pencil, parameterName=PencilPickaxe_Speed_2, defaultValue=<rgba -0.02 0.07 0 1>}.RGB) {g=true, r=true}
  n8 = VertexInterpolator(VS=PreSkinnedPosition())
  n9 = ComponentMask(None=Divide(A=@n8.PS) {constB=128}) {b=true, r=true}
  n10 = TextureObjectParameter() {
   group=Toon_Shading_Pencil
   parameterName=PencilPickaxe_TA_Texture
   samplerType=SAMPLERTYPE_Masks
   texture=/BRCosmetics/Weapons/FORT_Melee/Pickaxe_PencilCherry/Textures/T_Pickaxe_PencilCherry_TA
  }
  n11 = TextureSample(UVs=Add(A=Multiply(A=@n6, B=@n7), B=@n9), Tex=@n10) {
   samplerSource=SSM_Wrap_WorldGroupSettings
   samplerType=SAMPLERTYPE_Masks
   texture=/BRCosmetics/Characters/Player/Male/Large/Bodies/M_LRG_Clash_V/Textures/Simple_Perlin_M
  }
  n12 = Time()
  n13 = ComponentMask(None=VectorParameter() {group=Toon_Shading_Pencil, parameterName=PencilPickaxe_Speed, defaultValue=<rgba 0.05 -0.03 0 1>}.RGB) {g=true, r=true}
  n14 = TextureSample(UVs=Add(A=@n9, B=Multiply(A=@n12, B=@n13)), Tex=@n10) {
   samplerSource=SSM_Wrap_WorldGroupSettings
   samplerType=SAMPLERTYPE_Masks
   texture=&Simple_Perlin_M
  }
  n15 = Multiply(A=Multiply(A=@n11.G, B=@n14.G), B=ScalarParameter() {defaultValue=1.1, group=Toon_Shading_Pencil, parameterName=PencilPickaxe_NoiseStrength})
  n16 = ScalarParameter() {defaultValue=0.1, group=Toon_Shading_Pencil, parameterName=PencilPickaxe_NoiseBias}
  n17 = Divide(A=@n8.PS, B=Constant() {r=24}) {constB=24}
  n18 = TextureSample(
   UVs=MaterialFunctionCall(UVs=ComponentMask(None=@n17) {b=true, r=true}, "Rotation Angle (0-1)"=Constant() {r=0.023}) {materialFunction=&CustomRotator}.Rotated Values
   Tex=@n10
  ) {
   samplerSource=SSM_Wrap_WorldGroupSettings
   samplerType=SAMPLERTYPE_Masks
   texture=&Simple_Perlin_M
  }
  n19 = SmoothStep(
   Min=@n5
   Max=Add(A=@n5, B=ScalarParameter() {defaultValue=0.08, group=Toon_Shading_Pencil, parameterName=PencilPickaxe_GridSharpness})
   Value=Multiply(A=Subtract(A=@n15, B=@n16), B=SmoothStep(Value=@n18.R))
  ) {constMin=0.9}
  n20 = FunctionInput() {inputName=SkinFX}
  FunctionOutput(None=LinearInterpolate(
    A=NamedRerouteUsage()
    B=LinearInterpolate(A=@n1, B=@n4, Alpha=@n19)
    Alpha=ChannelMaskParameter(None=@n20) {group=Toon_Shading_Pencil, parameterName=PencilPickaxe_AnimatedChannel}.R
   ))
 ]}
