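# Toon shading material function for the Pencil cosmetic character, with switches for its backpack and pickaxe.
# Adjusted normals and a selectable light vector feed cinematic lighting and a curve-atlas ramp, alongside SSC threshold masks and lit/shaded mask colors.
# Fake or hair specular, eye highlights and distance-field ink lines layer on top.
# The result goes to Emissive on Cloth shading with black BaseColor; PBR_Shading blends it into BaseColor for real lighting.
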MaterialFunction /BRCosmetics/Characters/Player/Male/Medium/Bodies/M_MED_Pencil_A/PencilMaterials/MFC_AnimeShading_Pencil
graph {
 properties {libraryCategoriesText [Toon]}
 nodes [
  AtmosphericLightVector()
  n1 = VectorParameter() {group=Toon_Shading, parameterName=ScreenSpaceLightVector, defaultValue=<rgba 1 1 -1 1>}
  Composite!(LightVector=@n1.RGB) {subgraphName=ProjectScreenAlighned}
  n2 = StaticSwitchParameter(
   True=NamedRerouteUsage()
   False=StaticSwitchParameter(True=NamedRerouteUsage(), False=NamedRerouteUsage()) {group=Eye Shader, parameterName=Eye_Visualize Cornea, sortPriority=0}
  ) {group=Eye Shader, parameterName=Eye_Debug, sortPriority=1}
  n3 = NamedRerouteUsage()
  n4 = NamedRerouteUsage()
  n5 = NamedRerouteUsage()
  n6 = NamedRerouteUsage()
  n7 = ScalarParameter() {defaultValue=0.001, group=Toon_Shading, parameterName=Subsurface Mask Value, sliderMax=1}
  n8 = SetMaterialAttributes!(
   MaterialAttributes=NamedRerouteUsage()
   BaseColor=Constant3Vector() {desc="dont touch!"}
   Metallic=NamedRerouteUsage()
   Specular=Constant() {desc="dont touch!"}
   EmissiveColor=Add(
    A=NamedRerouteUsage()
    B=MaterialFunctionCall!(Emissive In=@n2, LightingExtrasColorData=@n3, CINE_LightSources=@n4, ShadingCurve=@n5) {materialFunction=/Game/Packages/Fortress_SharedMaterials/Cinematics/Anime/MF_CINE_FinalColorBlend}.Result
   )
   Roughness=LinearInterpolate(A=@n6, B=OneMinus(None=@n6), Alpha=ScalarParameter() {group=Toon_Shading_SSC, parameterName=Invert Roughness Pass, sliderMax=1})
   AmbientOcclusion=StaticSwitchParameter(True=NamedRerouteUsage(), False=Constant() {r=1}) {group=Toon_Shading, parameterName=UseDFInklines, sortPriority=0}
   SubsurfaceColor=MaterialFunctionCall(Y=@n7, Z=@n7) {materialFunction=&MakeFloat3}.Result
   CustomData0=Constant() {desc=Opacity, r=1}
   ShadingModel=ShadingModel() {shadingModel=MSM_Cloth}
   Normal=NamedRerouteUsage()
  ) {attributeSetTypes [
    #1
    #2
    #3
    #4
    #5
    #6
    #7
    #8
    #9
    #10
   ]}
  n9 = GetMaterialAttributes(None=@n8) {attributeGetTypes [#4, #3, #5]}
  n10 = SetMaterialAttributes!(
   MaterialAttributes=@n9.MaterialAttributes
   BaseColor=Saturate(None=@n9.EmissiveColor)
   EmissiveColor=Constant()
   AmbientOcclusion=Constant() {r=1}
   Metallic=Constant()
  ) {attributeSetTypes [
    #1
    #4
    #6
    #2
   ]}
  n11 = ChannelMaskParameter(None=VertexColor()) {group=Toon_Shading_Pencil, maskChannel=Green, parameterName=PencilPickaxe_WristPanelChannel, defaultValue=<rgba 0 1 0 0>}
  n12 = SetMaterialAttributes!(
   MaterialAttributes=@n10
   Specular=LinearInterpolate(
    A=@n9.Specular
    B=ScalarParameter() {defaultValue=1, group=Toon_Shading_Pencil, parameterName=PencilPickaxe_WristPanel_PBRSpecular}
    Alpha=@n11.R
   )
   Roughness=LinearInterpolate(A=@n9.Roughness, B=ScalarParameter() {group=Toon_Shading_Pencil, parameterName=PencilPickaxe_WristPanel_PBRRoughness}, Alpha=@n11.R) {constB=0}
  ) {attributeSetTypes [#3, #5]}
  n13 = Saturate(None=ScalarParameter() {group=Toon_Shading_Pencil, parameterName=PBR_Shading, sliderMax=1})
  FunctionOutput(None=BlendMaterialAttributes(
    A=@n8
    B=StaticSwitchParameter(True=@n12, False=@n10) {group=Toon_Shading_Pencil, parameterName=UsePickaxe, sortPriority=0}
    Alpha=NamedRerouteDeclaration(None=@n13) {name=PBR_Switch, nodeColor=<rgba 0 1 0.2227 1>}
   ))
  FunctionOutput(None=NamedRerouteUsage()) {outputName=SSC, sortPriority=3}
  FunctionOutput(None=NamedRerouteUsage()) {outputName=InkLine Mask, sortPriority=2}
  FunctionOutput(None=NamedRerouteUsage()) {outputName=PBR_Switch, sortPriority=5}
  FunctionOutput(None=NamedRerouteUsage()) {outputName=FX, sortPriority=4}
  FunctionOutput(None=NamedRerouteUsage()) {outputName=Shading Curve, sortPriority=1}
  n14 = TextureSampleParameter2D(UVs=NamedRerouteUsage()) {
   desc=ShadowGradient Texture
   group=Toon_Shading
   parameterName=SG_Texture
   samplerType=SAMPLERTYPE_LinearGrayscale
   sortPriority=1
   texture=/Game/Global/Textures/Default/Blanks/T_BlackLinearGrayscale
  }
  n15 = Multiply(A=@n14.R) {constB=2}
  n16 = MaterialFunctionCall(Float3=Transform(None=PixelNormalWS()) {transformSourceType=TRANSFORMSOURCE_World, transformType=TRANSFORM_Local}) {materialFunction=&BreakOutFloat3Components}
  n17 = Normalize(VectorInput=MaterialFunctionCall(X=Subtract(A=@n15), Y=@n16.G, Z=@n16.G) {materialFunction=&MakeFloat3}.Result)
  n18 = PixelNormalWS()
  n19 = Subtract(A=PreSkinnedPosition(), B=VectorParameter() {group=Toon_Shading, parameterName=NormalSphereLocation, defaultValue=<rgba 0 0 162 1>}.RGB)
  n20 = LinearInterpolate(
   A=@n19
   B=Multiply(A=@n19, B=Constant3Vector() {constant=<rgba 1 1 0 0>})
   Alpha=ScalarParameter() {group=Toon_Shading, parameterName=Sphere to Cylindrical Normals, sliderMax=1}
  )
  n21 = Transform(None=VertexInterpolator(VS=Normalize(VectorInput=@n20)).PS) {transformSourceType=TRANSFORMSOURCE_Local}
  n22 = ScalarParameter() {group=Toon_Shading, parameterName=NormalSphereBlend, sliderMax=1}
  n23 = Subtract(A=Multiply(A=VertexColor()) {constB=2})
  n24 = Normalize(VectorInput=LinearInterpolate(
    A=LinearInterpolate(A=@n18, B=@n21, Alpha=@n22) {constAlpha=0}
    B=Transform(None=@n23) {transformSourceType=TRANSFORMSOURCE_Local}
    Alpha=ScalarParameter() {group=Toon_Shading, parameterName=FakeNormalBlend_Amt, sliderMax=1}
   ))
  n25 = StaticSwitchParameter(
   True=LinearInterpolate(
    A=@n24
    B=Transform(None=@n17) {transformSourceType=TRANSFORMSOURCE_Local}
    Alpha=ScalarParameter() {defaultValue=0.5, desc=ShadowGradient Intensity Blending, group=Toon_Shading, parameterName=SG_Blend, sliderMax=1}
   )
   False=@n24
  ) {desc=Activate to use the ShadowGradient Map for the painted shading control., group=Toon_Shading, parameterName=UseSGMap, sortPriority=0}
  NamedRerouteDeclaration(None=@n25) {name=Adjusted Normals, nodeColor=<rgba 0.5 0.5 1 1>}
  n26 = NamedRerouteUsage()
  n27 = TextureSampleParameter2D(UVs=@n26) {
   group=Skin FX
   parameterName=SkinFX_Mask
   samplerType=SAMPLERTYPE_Masks
   sortPriority=1
   texture=&T_BlackMask
  }
  n28 = Add(A=Add(A=@n27.R, B=@n27.R), B=@n27.B)
  n29 = Saturate(None=Subtract(A=@n28, B=NamedRerouteUsage()))
  NamedRerouteDeclaration(None=OneMinus(None=StaticSwitchParameter(True=@n29, False=@n28) {group=Toon_Shading_Pencil, parameterName=UseBackpack})) {name=NotCustomColor_Mask, nodeColor=<rgba 0.7578 0 1 1>}
  n30 = VectorParameter() {group=Toon_Shading_Fake PBR, parameterName=Specular Tint, defaultValue=<rgba 1 1 1 1>}
  n31 = NamedRerouteUsage()
  n32 = NamedRerouteUsage()
  n33 = Multiply(
   A=LinearInterpolate(A=@n30.RGB, B=Multiply(A=@n30.RGB, B=@n31) {constA=1}, Alpha=NamedRerouteUsage())
   B=Multiply(
    A=ScalarParameter() {defaultValue=1, group=Toon_Shading_Fake PBR, parameterName=Specular Strength, sliderMax=1}
    B=LinearInterpolate(Alpha=@n32) {constA=1, constB=0.25}
   ) {constB=0.25}
  )
  n34 = ScalarParameter() {defaultValue=0.0005, group=Toon_Shading_Fake PBR, parameterName=Roughness Min, sliderMax=1}
  n35 = ScalarParameter() {defaultValue=1, group=Toon_Shading_Fake PBR, parameterName=Roughness Max, sliderMax=1}
  n36 = Saturate(None=Power(Base=NamedRerouteUsage()) {constExponent=3})
  n37 = NamedRerouteUsage()
  n38 = MaterialFunctionCall() {materialFunction=&CameraDirectionVector}
  n39 = Add(
   A=CameraVectorWS()
   B=LinearInterpolate(
    A=ComponentMask(None=@n37) {b=true, g=true, r=true}
    B=Multiply(A=@n38.Result) {constB=-1}
    Alpha=ScalarParameter() {group=Toon_Shading_Fake PBR, parameterName=CameraDirectionForSpecularHighlights, sliderMax=1}
   )
  )
  n40 = StaticSwitchParameter(
   True=ChannelMaskParameter(None=NamedRerouteUsage()) {group=Toon_Shading_Fake PBR, parameterName=Ignore Z SkinFX Channel, sortPriority=10}.R
   False=ScalarParameter() {group=Toon_Shading_Fake PBR, parameterName=Ignore Z, sliderMax=1}
  ) {group=Toon_Shading, parameterName=Ignore Z use SkinFX mask, sortPriority=0}
  n41 = AppendVector(A=Constant2Vector() {g=1, r=1}, B=Saturate(None=OneMinus(None=@n40)))
  n42 = NamedRerouteUsage()
  n43 = DotProduct(A=Normalize(VectorInput=Multiply(A=@n39, B=@n41)), B=Normalize(VectorInput=Multiply(A=@n41, B=@n42))) {desc=Blinn}
  n44 = ComponentMask(None=NamedRerouteUsage()) {b=true, g=true, r=true}
  n45 = DotProduct(A=ReflectionVectorWS(), B=Normalize(VectorInput=Multiply(A=@n41, B=@n44))) {desc=Phong}
  n46 = ScalarParameter() {desc=Interpolate between Blinn and Phong Highlight, group=Toon_Shading_Fake PBR, parameterName=BlendBlinnPhong, sliderMax=1}
  n47 = SmoothStep(
   Min=OneMinus(None=LinearInterpolate(A=@n34, B=@n35, Alpha=@n36))
   Max=Constant() {r=1}
   Value=Saturate(None=LinearInterpolate(A=@n43, B=@n45, Alpha=@n46))
  )
  n48 = LinearInterpolate(
   A=MaterialFunctionCall(In=@n47, Contrast=ScalarParameter() {defaultValue=2, group=Toon_Shading_Fake PBR, parameterName=Specular Sharpness, sliderMax=2}) {materialFunction=&CheapContrast}.Result
   B=Round(None=@n47)
   Alpha=ScalarParameter() {group=Toon_Shading_Fake PBR, parameterName=Bool Specular, sliderMax=1}
  )
  n49 = Multiply(A=Multiply(A=Multiply(A=@n33, B=@n48), B=NamedRerouteUsage()) {constA=1}, B=NamedRerouteUsage())
  n50 = NamedRerouteUsage()
  NamedRerouteDeclaration(None=Add(A=Multiply(A=@n49, B=@n50), B=NamedRerouteUsage())) {name=Fake PBR, nodeColor=<rgba 0 0.2266 1 1>}
  NamedRerouteDeclaration(None=@n27.RGBA) {name=SkinFX, nodeColor=<rgba 1 0.3281 0 1>}
  n51 = GetMaterialAttributes(None=FunctionInput() {inputName=Material, inputType=FunctionInput_MaterialAttributes}) {attributeGetTypes [#4, #10]}
  NamedRerouteDeclaration(None=@n51.EmissiveColor) {name=EmissiveColor, nodeColor=<rgba 1 1 0 1>}
  n52 = Constant3Vector() {constant=<rgba 0 0 1 0>}
  n53 = ChannelMaskParameter(None=NamedRerouteUsage()) {
   desc="Select in which channel of the FX map the mask camera light vector will be stored "
   group=Toon_Shading
   parameterName=CameraLightVector_FXChannel
   sortPriority=10
  }
  MaterialExpressionReroute_12 = Reroute(None=@n1.RGB)
  n54 = Normalize(VectorInput=@MaterialExpressionReroute_12.LightVector)
  n55 = ComponentMask(None=@n54) {r=true}
  n56 = Constant3Vector() {constant=<rgba 0 1 0 0>}
  n57 = Normalize(VectorInput=WorldPosition() {worldPositionShaderOffset=WPT_CameraRelative}.XYZ)
  n58 = Normalize(VectorInput=CrossProduct(A=Transform(None=@n56) {transformSourceType=TRANSFORMSOURCE_View}, B=@n57))
  n59 = ComponentMask(None=@n54) {g=true}
  n60 = CrossProduct(A=@n57, B=@n58)
  n61 = Add(A=Add(A=Multiply(A=@n55, B=@n58), B=Multiply(A=@n59, B=@n60)), B=Multiply(A=ComponentMask(None=@n54) {b=true}, B=@n57))
  MaterialExpressionReroute_14 = Reroute(None=@n61)
  n62 = Custom() {code="return View.DirectionalLightDirection;", description=DirectionalLightDirection, inputs []}
  n63 = CollectionParameter() {
   collection=&FortniteMaterialParameters
   parameterName=SunlightDirectionalVector
  }
  n64 = ComponentMask(None=@n63) {b=true, g=true, r=true}
  n65 = If(A=DotProduct(A=@n62, B=@n62), "A > B"=@n62, "A == B"=@n64, "A < B"=@n64)
  n66 = StaticSwitchParameter(
   True=@MaterialExpressionReroute_14.Result
   False=MaterialFunctionCall!(In=NamedRerouteDeclaration(None=@n65) {name=LightVector_Direct, nodeColor=<rgba 0 0.6953 1 1>}) {materialFunction=/Game/Packages/Fortress_SharedMaterials/Cinematics/Anime/MF_CINE_AnimeLightDirOverride}.Result
  ) {
   desc="Use a Screen / View alight light vector. \r\nOverrides the UseDirectionalLightFunction!"
   group=Toon_Shading
   parameterName=UseScreenSpaceLightDirection
   sortPriority=0
  }
  n67 = MaterialFunctionCall(Float3=@n66) {materialFunction=&BreakOutFloat3Components}
  n68 = MaterialFunctionCall(X=@n67.R, Y=@n67.R, Z=Constant()) {materialFunction=&MakeFloat3}
  n69 = ComponentMask(None=StaticSwitchParameter(True=Normalize(VectorInput=@n68.Result) {desc=need to transform it to preskinning}, False=@n66) {desc=Lock the Z axis of the Light Vector, group=Toon_Shading, parameterName=UseLockLightZVector, sortPriority=0}) {b=true, g=true, r=true}
  n70 = StaticSwitchParameter(
   True=LinearInterpolate(A=@n69, B=Transform(None=@n52) {desc=Camera Forward Vector, transformSourceType=TRANSFORMSOURCE_Camera}, Alpha=@n53.R)
   False=@n69
  ) {
   desc="Use the Camera vector instead of the light vector for an reaz which is defined by the FX map. Select the FX channel with this parameter \"CameraLightVector_FXChannel\"."
   group=Toon_Shading
   parameterName=UseCameraLightVector
   sortPriority=0
  }
  NamedRerouteDeclaration(None=@n70) {name=LightVector, nodeColor=<rgba 0 1 0.8634 1>}
  NamedRerouteDeclaration(None=@n51.MaterialAttributes) {name=MA, nodeColor=<rgba 1 1 0 1>}
  n71 = NamedRerouteUsage()
  n72 = Power(Base=@n71, Exp=ScalarParameter() {defaultValue=4, group=Toon_Shading_DF_InkLines, parameterName=DF_BlurPow, sliderMax=16}) {constExponent=3}
  n73 = ScalarParameter() {defaultValue=1, group=Toon_Shading_DF_InkLines, parameterName=DF_BlurMult, sliderMax=2}
  n74 = WorldPosition()
  n75 = CameraPositionWS()
  n76 = ScalarParameter() {defaultValue=100, group=Toon_Shading_DF_InkLines, parameterName=DF_FarDistance, sliderMax=1000}
  n77 = Divide(
   A=Subtract(A=Distance(A=@n74.XYZ, B=@n75), B=@n76)
   B=Subtract(A=ScalarParameter() {defaultValue=500, group=Toon_Shading_DF_InkLines, parameterName=DF_NearDistance, sliderMax=1000}, B=@n76)
  )
  n78 = SphereMask(
   A=@n71
   B=Constant() {r=1}
   Radius=LinearInterpolate(
    A=ScalarParameter() {defaultValue=0.4, group=Toon_Shading_DF_InkLines, parameterName=DF_LineThickness_Near, sliderMax=1}
    B=ScalarParameter() {defaultValue=0.6, group=Toon_Shading_DF_InkLines, parameterName=DF_LineThickness_Far, sliderMax=1}
    Alpha=Saturate(None=@n77)
   )
   Hardness=ScalarParameter() {defaultValue=0.8, group=Toon_Shading_DF_InkLines, parameterName=DF_LineHardness, sliderMax=1}
  ) {attenuationRadius=0.1}
  n79 = Multiply(A=OneMinus(None=Multiply(A=@n72, B=@n73) {constB=0.5}), B=OneMinus(None=Saturate(None=@n78)))
  NamedRerouteDeclaration(None=Multiply(A=Saturate(None=@n79), B=ScalarParameter() {defaultValue=1, group=Toon_Shading_DF_InkLines, parameterName=DF_Alpha, sliderMax=1})) {name=InkLine Mask, nodeColor=<rgba 0.1615 0.1615 0.1615 1>}
  n80 = NamedRerouteUsage()
  n81 = MaterialFunctionCall!(In=@n80, LUT, LUT Multiplier) {materialFunction=/Game/Packages/Fortress_SharedMaterials/Anime_Master_Material/MF_ApplyLUT}
  n82 = Multiply(
   A=NamedRerouteUsage()
   B=ScalarParameter() {defaultValue=0.5, group=Toon_Shading_DF_InkLines, parameterName=InkLinerBaseColorStrenght, sliderMax=1}
  )
  n83 = VectorParameter() {group=Toon_Shading_DF_InkLines, parameterName=InkLineColor, sortPriority=10, defaultValue=<rgba 0 0 0 1>}
  NamedRerouteDeclaration(None=StaticSwitchParameter(
    True=StaticSwitchParameter(True=@n81.Result, False=@n80) {group=Toon_Shading, parameterName=Use Apply LUT, sortPriority=0}
    False=StaticSwitchParameter(True=@n82, False=@n83.RGB) {group=Toon_Shading, parameterName=UseBaseColorAsDFLineColor, sortPriority=0}
   ) {group=Toon_Shading, parameterName=UseInkLineColorTexture, sortPriority=0}) {name=InkLine Color, nodeColor=<rgba 0.1615 0.1615 0.1615 1>}
  n84 = TextureSampleParameter2D(UVs=@n26) {group=Toon_Shading, parameterName=LitDiffuse, sortPriority=1, texture=&T_WhitesRGB}
  n85 = MaterialFunctionCall!(BaseColor_IN=@n84.RGB, SkinFX=NamedRerouteUsage()) {materialFunction=&/MFC_Chibi}
  NamedRerouteDeclaration(None=MaterialFunctionCall!(LitBaseColorIn=StaticSwitchParameter(True=@n85.Chibi, False=@n84.RGB) {group=Toon_Shading_Pencil, parameterName=UseBackpack}) {materialFunction=/Game/Packages/Fortress_SharedMaterials/Cinematics/Anime/MF_CINE_LitDiffuseTint}.Result) {name=Map_BaseColor_Lit, nodeColor=<rgba 0 1 0.7863 1>}
  NamedRerouteDeclaration(None=FunctionInput(Preview=TextureCoordinate()) {bUsePreviewValueAsDefault=true, inputName=CustomUVs, inputType=FunctionInput_Vector2, sortPriority=15}) {name=CustomUVs, nodeColor=<rgba 1 0.0993 0 1>}
  n86 = VectorParameter() {group=Toon_Shading_Pencil, parameterName=ColorLitDefault_R_Mask, sortPriority=23, defaultValue=<rgba 1 0 0 0>}
  n87 = VectorParameter() {
   desc="if A=0.0 ColorLitDefault is used"
   group=Toon_Shading_Pencil
   parameterName=ColorLit_R_Mask
   sortPriority=15
   defaultValue=<rgba 1 0 0 0>
  }
  n88 = MaterialFunctionCall(Float3=NamedRerouteUsage()) {materialFunction=&BreakOutFloat3Components}
  n89 = VectorParameter() {group=Toon_Shading_Pencil, parameterName=ColorLitDefault_G_Mask, sortPriority=24, defaultValue=<rgba 0 1 0 0>}
  n90 = VectorParameter() {
   desc="if A=0.0 ColorLitDefault is used"
   group=Toon_Shading_Pencil
   parameterName=ColorLit_G_Mask
   sortPriority=16
   defaultValue=<rgba 0 1 0 0>
  }
  n91 = Add(
   A=Multiply(A=LinearInterpolate(A=@n86.RGB, B=@n87.RGB, Alpha=@n87.RGB), B=@n88.R)
   B=Multiply(A=LinearInterpolate(A=@n89.RGB, B=@n90.RGB, Alpha=@n90.RGB), B=@n88.G)
  )
  n92 = VectorParameter() {group=Toon_Shading_Pencil, parameterName=ColorLitDefault_B_Mask, sortPriority=25, defaultValue=<rgba 0 0 1 0>}
  n93 = VectorParameter() {
   desc="if A=0.0 ColorLitDefault is used"
   group=Toon_Shading_Pencil
   parameterName=ColorLit_B_Mask
   sortPriority=17
   defaultValue=<rgba 0 0 1 0>
  }
  n94 = Multiply(A=LinearInterpolate(A=@n92.RGB, B=@n93.RGB, Alpha=@n93.RGB), B=@n88.B)
  NamedRerouteDeclaration(None=MaterialFunctionCall(LitBaseColorIn=Add(A=@n91, B=@n94)) {materialFunction=&MF_CINE_LitDiffuseTint}.Result) {name=CustomColorsLit, nodeColor=<rgba 0.8281 1 0 1>}
  n95 = TextureSampleParameter2D(UVs=NamedRerouteUsage()) {
   desc="R = SpecMask G = Metallic B = Roughness"
   group=Specular
   parameterName=SpecularMasks
   samplerType=SAMPLERTYPE_Masks
   texture=&T_Fortnite_Default_S
  }
  NamedRerouteDeclaration(None=@n95.B) {name=Roughness Input, nodeColor=<rgba 1 1 0 1>}
  n96 = TextureSampleParameter2D(UVs=@n26) {
   group=Toon_Shading
   parameterName=DistanceField_InkLines
   samplerSource=SSM_Wrap_WorldGroupSettings
   samplerType=SAMPLERTYPE_Alpha
   sortPriority=1
   texture=/Game/Global/Textures/Default/Blanks/T_BlackAlpha
  }
  NamedRerouteDeclaration(None=@n96.RGBA) {name=Map_DF_Inklines, nodeColor=<rgba 0.2865 0.2865 0.2865 1>}
  n97 = TextureSampleParameter2D(UVs=@n26) {group=Toon_Shading, parameterName=InkLineColor_Texture, sortPriority=1, texture=&T_BlacksRGB}
  NamedRerouteDeclaration(None=@n97.RGB) {name=Map_DF_Inklines_Color, nodeColor=<rgba 0.0432 1 0 1>}
  n98 = TextureSampleParameter2D(UVs=@n26) {
   group=Toon_Shading
   parameterName=SSC_Texture
   samplerSource=SSM_Wrap_WorldGroupSettings
   samplerType=SAMPLERTYPE_Masks
   sortPriority=1
   texture=/Game/Global/Textures/Default/Blanks/T_Blank_Mask_Magenta_R1G0B1A1
  }
  NamedRerouteDeclaration(None=@n98.RGBA) {name=Map_SSC, nodeColor=<rgba 1 0.3207 0.7242 1>}
  n99 = TextureSampleParameter2D(UVs=@n26) {
   group=Toon_Shading
   parameterName=SSC_Texture_Far
   samplerSource=SSM_Wrap_WorldGroupSettings
   samplerType=SAMPLERTYPE_Masks
   sortPriority=1
   texture=&T_Blank_Mask_Magenta_R1G0B1A1
  }
  NamedRerouteDeclaration(None=@n99.RGBA) {name=Map_SSC_Far, nodeColor=<rgba 1 0.3207 0.7242 1>}
  n100 = NamedRerouteUsage()
  n101 = LinearInterpolate(A=ScalarParameter() {defaultValue=0.5, group=Toon_Shading, parameterName=ThresholdDarkeningValue, sliderMax=1}, Alpha=NamedRerouteUsage())
  n102 = NamedRerouteUsage()
  n103 = StaticSwitchParameter(
   True=MaterialFunctionCall!(SkinFX=NamedRerouteUsage(), CustomColorsLit=@n102) {materialFunction=&/MFC_PencilPickaxe}.Result
   False=@n102
  ) {group=Toon_Shading_Pencil, parameterName=UsePickaxe, sortPriority=0}
  n104 = Multiply(A=NamedRerouteUsage(), B=@n103)
  n105 = LinearInterpolate(A=NamedRerouteUsage(), Alpha=NamedRerouteUsage())
  n106 = NamedRerouteUsage()
  n107 = NamedRerouteUsage()
  n108 = Multiply(A=ScalarParameter() {defaultValue=0.5, group=Toon_Shading, parameterName=ShadedColorDarkening, sliderMax=1}, B=@n107)
  n109 = LinearInterpolate(
   A=Multiply(A=NamedRerouteUsage(), B=LinearInterpolate(A=@n104, B=@n103, Alpha=@n105))
   B=LinearInterpolate(A=StaticSwitchParameter(True=@n106, False=@n108) {group=Toon_Shading, parameterName=UseShadingTexture, sortPriority=0}, B=@n107, Alpha=@n105)
   Alpha=NamedRerouteUsage()
  )
  n110 = VectorParameter() {group=Toon_Shading_Pencil, parameterName=ColorDifShaded_Skin, sortPriority=27, defaultValue=<rgba 0.5 0.5 0.5 0>}
  n111 = VectorParameter() {group=Toon_Shading_Pencil, parameterName=ColorLit_Skin, sortPriority=26, defaultValue=<rgba 0.9 0.5 0.3 1>}
  n112 = LinearInterpolate(A=Multiply(A=@n110.RGB, B=@n111.RGB), B=@n111.RGB, Alpha=@n105)
  n113 = NamedRerouteUsage()
  n114 = Saturate(None=StaticSwitchParameter(True=LinearInterpolate(A=@n109, B=@n112, Alpha=@n113), False=@n109) {group=Toon_Shading, parameterName=UseVCGreenAsSkinMask, sortPriority=0})
  n115 = LinearInterpolate(A=LinearInterpolate(A=Multiply(A=@n101, B=@n114), B=@n114, Alpha=NamedRerouteUsage()), B=@n114, Alpha=NamedRerouteUsage())
  n116 = MaterialFunctionCall(In=MaterialFunctionCall!(In=@n115) {materialFunction=/Game/Packages/Fortress_SharedMaterials/Anime_Master_Material/MF_TODCustomColors}.Result) {materialFunction=&MF_ApplyLUT}
  n117 = Multiply(
   A=ScalarParameter() {defaultValue=0.8, group=Toon_Shading_Pencil, parameterName=PBR_CompensationScale, sliderMax=1}
   B=MaterialFunctionCall(In=@n115) {materialFunction=&MF_ApplyLUT}.Result
  ) {constB=0.5}
  n118 = NamedRerouteUsage()
  n119 = NamedRerouteDeclaration(None=StaticSwitchParameter(True=LinearInterpolate(A=@n116.Result, B=@n117, Alpha=@n118), False=@n115) {group=Toon_Shading, parameterName=Use Apply LUT, sortPriority=0}) {name=BaseColorOUT, nodeColor=<rgba 0 1 0.5705 1>}
  n120 = NamedRerouteUsage()
  n121 = Multiply(
   A=Multiply(A=@n119, B=VectorParameter() {group=Toon_Shading_Hair, parameterName=SpecularColor, defaultValue=<rgba 1 1 1 1>}.RGB)
   B=Multiply(
    A=ScalarParameter() {defaultValue=10, group=Toon_Shading_Hair, parameterName=SpecularStrength, sliderMax=20}
    B=LinearInterpolate(Alpha=@n120) {constA=1, constB=0.75}
   ) {constB=0.25}
  )
  n122 = NamedRerouteUsage()
  n123 = ScalarParameter() {group=Toon_Shading_Hair, parameterName=IgnoreZforSpecularHighlight, sliderMax=1}
  n124 = AppendVector(A=Constant2Vector() {g=1, r=1}, B=Saturate(None=OneMinus(None=@n123)))
  n125 = Multiply(A=NamedRerouteUsage(), B=@n124)
  n126 = Multiply(A=MaterialFunctionCall() {materialFunction=&CameraDirectionVector}.Result) {constB=-1}
  n127 = NamedRerouteUsage()
  n128 = Multiply(
   A=Add(
    A=CameraVectorWS()
    B=StaticSwitchParameter(True=@n126, False=@n127) {group=Toon_Shading_Hair, parameterName=UseCameraDirectionForSpecularHighlights, sortPriority=0}
   )
   B=@n124
  )
  n129 = Multiply(A=@n124, B=PixelNormalWS())
  n130 = StaticSwitchParameter(
   True=DotProduct(A=ReflectionVectorWS(), B=Normalize(VectorInput=@n125)) {desc=Phong}
   False=DotProduct(A=Normalize(VectorInput=@n128), B=Normalize(VectorInput=@n129)) {desc=Blinn}
  ) {desc=Switches between Blinn and Phong Highlight., group=Toon_Shading_Hair, parameterName=UsePhongHighlight, sortPriority=0}
  n131 = MaterialFunctionCall(
   In=Power(Base=Saturate(None=@n130), Exp=ScalarParameter() {defaultValue=50, group=Toon_Shading_Hair, parameterName=SpecPower, sliderMax=32})
   Contrast=ScalarParameter() {defaultValue=4, group=Toon_Shading_Hair, parameterName=Sharpness, sliderMax=8}
  ) {materialFunction=&CheapContrast}
  n132 = ScalarParameter() {
   defaultValue=0.2
   desc=Instensity of the specular highlight in the shaded area.
   group=Toon_Shading_Hair
   parameterName=ShadedSpecularHighlightIntensity
   sliderMax=1
  }
  n133 = NamedRerouteUsage()
  n134 = StaticSwitchParameter(True=Multiply(A=Multiply(A=@n122, B=@n131.Result), B=LinearInterpolate(A=@n132, Alpha=@n133)), False=@n131.Result) {
   desc=Disables the Anisotropy setup and used a Fresnel to reveal the ssc B mask. adjust the specular color.
   group=Toon_Shading_Hair
   parameterName=UseSSCBlueasHighlight
   sortPriority=0
  }
  n135 = LinearInterpolate(
   A=@n119
   B=StaticSwitchParameter(True=LinearInterpolate(A=@n119, B=@n121, Alpha=@n134), False=@n119) {group=Toon_Shading, parameterName=UseSpecularHighlight, sortPriority=0}
   Alpha=NamedRerouteUsage()
  )
  n136 = StaticSwitchParameter(
   True=NamedRerouteUsage()
   False=NamedRerouteDeclaration(None=StaticSwitchParameter(True=@n100, False=@n135) {group=Toon_Shading, parameterName=UseFakePBR, sortPriority=0}) {name=ReactiveColor_IN, nodeColor=<rgba 0 1 0.9531 1>}
  ) {group=Toon_Shading_Pencil, parameterName=UseBackpack}
  n137 = StaticSwitchParameter(True=Add(A=@n136, B=NamedRerouteUsage()), False=@n136) {group=Eye Shader, parameterName=UseEyeHighlight, sortPriority=-1}
  n138 = StaticSwitchParameter(True=LinearInterpolate(A=NamedRerouteUsage(), B=@n137, Alpha=NamedRerouteUsage()), False=@n137) {group=Toon_Shading, parameterName=UseDFInklines, sortPriority=0}
  n139 = ScalarParameter() {defaultValue=1, group=Toon_Shading, parameterName=GlobalBrightness, sliderMax=2}
  NamedRerouteDeclaration(None=MaterialFunctionCall!(Color=Multiply(A=@n138, B=@n139)) {materialFunction=/Game/Packages/Fortress_SharedMaterials/Anime_Master_Material/MF_TODColorAdjustment}.Result) {name=BC_Final, nodeColor=<rgba 0.4062 0 1 1>}
  n140 = NamedRerouteUsage()
  n141 = NamedRerouteUsage()
  n142 = ScalarParameter() {defaultValue=100, group=Toon_Shading_SSC, parameterName=SSC_FarDistance, sliderMax=1000}
  n143 = Subtract(A=Distance(A=WorldPosition().XYZ, B=CameraPositionWS()), B=@n142)
  n144 = Subtract(A=ScalarParameter() {defaultValue=500, group=Toon_Shading_SSC, parameterName=SSC_NearDistance, sliderMax=1000}, B=@n142)
  n145 = Divide(
   A=ComponentMask(None=ViewProperty() {property=MEVP_BufferSize}.Property) {r=true}
   B=ScalarParameter() {defaultValue=3840, group=Toon_Shading_SSC, parameterName=Reference Screen Size, sliderMax=3840}
  )
  n146 = Saturate(None=Divide(A=Divide(A=@n143, B=@n144), B=Power(Base=@n145) {constExponent=1.2}))
  n147 = NamedRerouteDeclaration(None=StaticSwitchParameter(True=LinearInterpolate(A=@n140, B=@n141, Alpha=@n146), False=@n140) {group=Toon_Shading, parameterName=Enable Far SSC, sortPriority=0}) {name=SSC, nodeColor=<rgba 1 0 0.4453 1>}
  NamedRerouteDeclaration(None=ChannelMaskParameter(None=@n147) {group=Toon_Shading_SSC, parameterName=SSC_ThresholdChannel, sortPriority=10}.R) {name=SSC_Threshold_Red, nodeColor=<rgba 1 0 0.0472 1>}
  NamedRerouteDeclaration(None=@n95.R) {name=Specular Input, nodeColor=<rgba 1 1 0 1>}
  n148 = ChannelMaskParameter(None=@n147) {group=Toon_Shading_SSC, maskChannel=Green, parameterName=SSC_OverlapLineValueChannel, sortPriority=10, defaultValue=<rgba 0 1 0 0>}
  NamedRerouteDeclaration(None=@n148.R) {name=SSC_Overlap_Green, nodeColor=<rgba 0 1 0.0597 1>}
  NamedRerouteDeclaration(None=VertexColor().G) {name=SkinMask, nodeColor=<rgba 0 1 0.0121 1>}
  NamedRerouteDeclaration(None=@n95.G) {name=Metallic Input, nodeColor=<rgba 1 1 0 1>}
  NamedRerouteDeclaration(None=FunctionInput() {inputName=EyeDebug, sortPriority=18}) {name=EyeDebug, nodeColor=<rgba 0 1 0.4141 1>}
  NamedRerouteDeclaration(None=FunctionInput() {inputName=DebugCornea, inputType=FunctionInput_Vector2, sortPriority=19}) {name=DebugCornea, nodeColor=<rgba 0.2422 1 0 1>}
  NamedRerouteDeclaration(None=LinearInterpolate(A=NamedRerouteUsage(), B=Constant3Vector() {constant=<rgba 1 0 0 0>}, Alpha=NamedRerouteUsage())) {name=VisualizeCornea, nodeColor=<rgba 1 0.3281 0 1>}
  n149 = MaterialFunctionCall(
   Normal=NamedRerouteUsage()
   Flatness=ScalarParameter() {defaultValue=0.25, group=Eye Shader, parameterName=Eye_Iris Normal Flatten, sliderMax=1, sortPriority=21}
  ) {materialFunction=&FlattenNormal}
  n150 = NamedRerouteUsage()
  n151 = ScalarParameter() {defaultValue=1, desc=Blend modified normal with original normal, group=Eye Shader, parameterName=Eye_Blend_Normal, sliderMax=2}
  NamedRerouteDeclaration(None=LinearInterpolate(
    A=NamedRerouteUsage()
    B=MaterialFunctionCall(BaseNormal=@n149.Result, AdditionalNormal=@n150) {materialFunction=&BlendAngleCorrectedNormals}.Result
    Alpha=@n151
   )) {name=Normal_OUT, nodeColor=<rgba 0 1 0.9062 1>}
  NamedRerouteDeclaration(None=@n85.ChibiAlpha) {name=ChibiAlpha, nodeColor=<rgba 0 0.8125 1 1>}
  NamedRerouteDeclaration(None=FunctionInput() {bUsePreviewValueAsDefault=true, inputName=EyeNormal, sortPriority=16, previewValue=<0 0 1 1>}) {name=EyeNormal, nodeColor=<rgba 0.4766 1 0 1>}
  NamedRerouteDeclaration(None=@n51.Normal) {name=Normal_IN, nodeColor=<rgba 0.4766 0 1 1>}
  NamedRerouteDeclaration(None=@n85.ReactivityAlpha) {name=ReactivityAlpha, nodeColor=<rgba 0.8984 0 1 1>}
  n152 = MaterialFunctionCall(
   Alpha=Power(Base=NamedRerouteUsage(), Exp=ScalarParameter() {defaultValue=2, group=Toon_Shading, parameterName=ThresholdMaskPower, sliderMax=8})
   Min=ScalarParameter() {group=Toon_Shading, parameterName=Shaded_Threshold_Min, sliderMax=1}
   Max=ScalarParameter() {defaultValue=1, group=Toon_Shading, parameterName=Shaded_Threshold_Max, sliderMax=1}
  ) {materialFunction=&SmoothStep}
  NamedRerouteDeclaration(None=Round(None=@n152.Result)) {name=SSC Throshold, nodeColor=<rgba 0.7973 1 0 1>}
  n153 = Multiply(A=NamedRerouteUsage(), B=NamedRerouteUsage())
  n154 = Desaturation(
   None=MaterialFunctionCall(
    Hue Shift Percentage=ScalarParameter() {defaultValue=-0.1, group=Toon_Shading_Pencil, parameterName=PencilBackpack_Reactive_Hueshift}
    Texture=NamedRerouteDeclaration(None=@n153) {name=CustomColor_Screen, nodeColor=<rgba 1 0.6328 0 1>}
   ) {materialFunction=&HueShift}.Result
   Fraction=ScalarParameter() {defaultValue=-2, group=Toon_Shading_Pencil, parameterName=PencilBackpack_Reactive_Saturation}
  )
  n155 = ScalarParameter() {defaultValue=4, group=Toon_Shading_Pencil, parameterName=PencilBackpack_Reactive_Brightness}
  NamedRerouteDeclaration(None=LinearInterpolate(A=NamedRerouteUsage(), B=Multiply(A=@n154, B=@n155), Alpha=NamedRerouteUsage())) {name=ReactiveColor_OUT, nodeColor=<rgba 0 0.8359 1 1>}
  n156 = MaterialFunctionCall!(In=DotProduct(A=NamedRerouteUsage(), B=NamedRerouteUsage())) {materialFunction=/Game/Packages/Fortress_SharedMaterials/Cinematics/Anime/MF_CINE_AnimeLightingFeatures}
  NamedRerouteDeclaration(None=@n156.CINE_LightSources) {name=CINE_LightSources, nodeColor=<rgba 1 0.9023 0 1>}
  NamedRerouteDeclaration(None=@n156.CINE_LightColor) {name=CINE_LightColors, nodeColor=<rgba 1 0.8464 0 1>}
  n157 = Saturate(None=@n156.CINE_ShadingCurve)
  NamedRerouteDeclaration(None=Round(None=@n157)) {name=Shading Dot, nodeColor=<rgba 0.7973 1 0 1>}
  n158 = NamedRerouteUsage()
  n159 = NamedRerouteUsage()
  NamedRerouteDeclaration(None=StaticSwitchParameter(
    True=NamedRerouteUsage()
    False=StaticSwitchParameter(True=@n158, False=@n159) {group=Eye Shader, parameterName=Eye Use Sun Highlight, sortPriority=1}
   ) {defaultValue=true, group=Eye Shader, parameterName=Eye Use UV Highlight, sortPriority=1}) {name=Eye Highlight, nodeColor=<rgba 0.5938 0 1 1>}
  n160 = Multiply(A=NamedRerouteUsage(), B=NamedRerouteUsage())
  n161 = ComponentMask(None=VectorParameter() {group=Eye Shader, parameterName=Eye UV Highlight Pos, defaultValue=<rgba 0.92 0.12 0 0>}.RGB) {g=true, r=true}
  n162 = NamedRerouteUsage()
  n163 = ScalarParameter() {defaultValue=0.005, group=Eye Shader, parameterName=Eye UV Highlight Size}
  n164 = ScalarParameter() {defaultValue=0.6, group=Eye Shader, parameterName=Eye UV Highlight Hardness, sliderMax=1}
  n165 = Multiply(
   A=Multiply(
    A=VectorParameter() {group=Eye Shader, parameterName=Eye UV Highlight Color Tint, defaultValue=<rgba 1 1 1 1>}.RGB
    B=ScalarParameter() {defaultValue=1, group=Eye Shader, parameterName=Eye UV Highlight Strength}
   )
   B=SphereMask(A=@n160, B=Multiply(A=@n161, B=@n162), Radius=@n163, Hardness=@n164)
  )
  n166 = ComponentMask(None=VectorParameter() {group=Eye Shader, parameterName=Eye UV Highlight Pos 2, defaultValue=<rgba 0.878 0.12 0 0>}.RGB) {g=true, r=true}
  n167 = NamedRerouteUsage()
  n168 = Multiply(
   A=SphereMask(A=@n160, B=Multiply(A=@n166, B=@n167), Radius=@n163, Hardness=@n164)
   B=Multiply(
    A=ScalarParameter() {defaultValue=1, group=Eye Shader, parameterName=Eye UV Highlight 2 Strength}
    B=VectorParameter() {group=Eye Shader, parameterName=Eye UV Highlight 2 Color Tint, defaultValue=<rgba 1 1 1 1>}.RGB
   )
  )
  NamedRerouteDeclaration(None=StaticSwitchParameter(True=Add(A=@n165, B=@n168), False=@n165) {group=Eye Shader, parameterName=Add Second UV Highlight, sortPriority=1}) {name=UV Highlight, nodeColor=<rgba 0.7812 0 1 1>}
  n169 = Normalize(VectorInput=Add(A=CameraVectorWS(), B=NamedRerouteUsage()))
  n170 = Normalize(VectorInput=VertexNormalWS())
  NamedRerouteDeclaration(None=SmoothStep(Value=DotProduct(A=@n169, B=@n170) {desc=Blinn}) {constMin=0.995}) {name=SunDirectionHighlight, nodeColor=<rgba 1 0.75 0 1>}
  n171 = VectorParameter() {group=Eye Shader, parameterName=Eye Camera Light Vector, sortPriority=36, defaultValue=<rgba -0.2 0.075 -0.5 0>}
  n172 = Transform(None=Normalize(VectorInput=@n171.RGB)) {transformSourceType=TRANSFORMSOURCE_View}
  NamedRerouteDeclaration(None=SmoothStep(Value=DotProduct(A=@n170, B=@n172)) {constMin=0.995}) {name=Camera Highlight, nodeColor=<rgba 1 0.75 0 1>}
  n173 = Add(A=@n156.CINE_ShadingCurve, B=ScalarParameter() {group=Toon_Shading, parameterName=Lit Bias, sliderMax=1})
  n174 = ScalarParameter() {defaultValue=12, group=Toon_Shading_RimHighlighBacklit, parameterName=BackLitRim_OutlineC2, sliderMax=16, sliderMin=-8}
  n175 = Fresnel(
   ExponentIn=ScalarParameter() {defaultValue=3, group=Toon_Shading_RimHighlighBacklit, parameterName=BackLitExponent1, sliderMax=8}
   Normal=NamedRerouteUsage()
  ) {baseReflectFraction=0}
  n176 = LinearInterpolate(
   A=ScalarParameter() {defaultValue=-4, group=Toon_Shading_RimHighlighBacklit, parameterName=BackLitRim_OutlineC1, sliderMax=16}
   B=@n174
   Alpha=Power(
    Base=OneMinus(None=@n175)
    Exp=ScalarParameter() {defaultValue=4, group=Toon_Shading_RimHighlighBacklit, parameterName=BacklitOutlineExponent2, sliderMax=8}
   )
  )
  n177 = Add(A=OneMinus(None=Saturate(None=@n176)), B=@n157)
  n178 = Add(
   A=DotProduct(
    A=MaterialFunctionCall() {materialFunction=&CameraDirectionVector}.Result
    B=NamedRerouteUsage()
   )
   B=ScalarParameter() {defaultValue=0.25, group=Toon_Shading_RimHighlighBacklit, parameterName=BackLitBias, sliderMax=1}
  )
  n179 = ScalarParameter() {defaultValue=0.75, group=Toon_Shading_RimHighlighBacklit, parameterName=BackLitMaxBias, sliderMax=1}
  n180 = CurveAtlasRowParameter(CurveTime=LinearInterpolate(A=Multiply(A=@n173, B=@n152.Result), B=Multiply(A=@n177, B=@n152.Result), Alpha=Clamp(None=@n178, Max=@n179))) {
   atlas=/Game/Characters/Shared/CurveAtlas/AthenaCharacterCurveAtlas
   curve=/Game/Packages/Fortress_SharedMaterials/Anime_Master_Material/TwoToneShadingCurve
   defaultValue=14
   group=Toon_Shading
   parameterName=ShadingCurve
   sortPriority=5
  }
  NamedRerouteDeclaration(None=ChannelMaskParameter(None=@n180) {group=Toon_Shading, parameterName=ShadingCurveChannel, sortPriority=10}.R) {name=Shading Curve, nodeColor=<rgba 0.7973 1 0 1>}
  n181 = ScalarParameter() {defaultValue=2, desc="Texture Width/Height (2048x1024 would be 2)", group=Eye Shader, parameterName=Eye Texture AspectRatio, sortPriority=1}
  NamedRerouteDeclaration(None=MaterialFunctionCall(X=Constant() {r=1}, Y=Divide(B=@n181)) {materialFunction=&MakeFloat2}.Result) {name=AspectRatio, nodeColor=<rgba 1 0.4688 0 1>}
  NamedRerouteDeclaration(None=FunctionInput() {bUsePreviewValueAsDefault=true, inputName=UV, inputType=FunctionInput_Vector2, sortPriority=17}) {name=UV_In, nodeColor=<rgba 0.0312 0 1 1>}
  n182 = VectorParameter() {group=Toon_Shading_Pencil, parameterName=ColorDifShaded_R_Mask, sortPriority=20, defaultValue=<rgba 0.5 0.5 0.5 0>}
  n183 = VectorParameter() {group=Toon_Shading_Pencil, parameterName=ColorDifShaded_G_Mask, sortPriority=21, defaultValue=<rgba 0.5 0.5 0.5 0>}
  n184 = Add(A=Multiply(A=@n182.RGB, B=@n88.R), B=Multiply(A=@n183.RGB, B=@n88.G))
  n185 = VectorParameter() {group=Toon_Shading_Pencil, parameterName=ColorDifShaded_B_Mask, sortPriority=22, defaultValue=<rgba 0.5 0.5 0.5 0>}
  n186 = Multiply(A=@n185.RGB, B=@n88.B)
  NamedRerouteDeclaration(None=MaterialFunctionCall!(ShadedBaseColorIn=Add(A=@n184, B=@n186)) {materialFunction=/Game/Packages/Fortress_SharedMaterials/Cinematics/Anime/MF_CINE_ShadedDiffuseTint}.Result) {name=CustomColorsDifShaded, nodeColor=<rgba 0.2804 0.3385 0 1>}
  n187 = ChannelMaskParameter(None=@n147) {group=Toon_Shading_SSC, maskChannel=Blue, parameterName=SSC_SheenMaskChannel, sortPriority=10, defaultValue=<rgba 0 0 1 0>}
  NamedRerouteDeclaration(None=@n187.R) {name=SSC_Sheen_Blue, nodeColor=<rgba 0 0.1118 1 1>}
  n188 = TextureSampleParameter2D(UVs=@n26) {group=Toon_Shading, parameterName=ShadedDiffuse, sortPriority=1, texture=/Game/Global/Textures/Default/Blanks/T_MidgreysRGB}
  NamedRerouteDeclaration(None=MaterialFunctionCall(ShadedBaseColorIn=@n188.RGB) {materialFunction=&MF_CINE_ShadedDiffuseTint}.Result) {name=Map_BaseColor_Shaded, nodeColor=<rgba 0 0.3333 0.2621 1>}
  PinBase() {reroutePins [LightVector {expression=&self:MaterialExpressionReroute_12}]}
  PinBase!(Result=@n61) {reroutePins [Result {expression=&self:MaterialExpressionReroute_14}]}
  SkyAtmosphereLightDirection()
  n189 = ScalarParameter() {group=Cinematics, parameterName=CIN_Shaded_Desat, sliderMax=2}
  n190 = Reroute()
  StaticSwitchParameter(
   True=Multiply(
    A=VectorParameter() {group=Cinematics, parameterName=CIN_FillTint, defaultValue=<rgba 1 1 1 1>}.RGB
    B=Desaturation(None=@n190, Fraction=@n189)
   )
   False=@n190
  ) {group=Cinematics, parameterName=CIN_TintOverride, sortPriority=0}
  n191 = Reroute()
  StaticSwitchParameter(True=Multiply(A=@n191, B=VectorParameter() {group=Cinematics, parameterName=CIN_KeyTint, defaultValue=<rgba 1 1 1 1>}.RGB), False=@n191) {group=Cinematics, parameterName=CIN_TintOverride, sortPriority=0}
  VertexNormalWS()
 ]
}
