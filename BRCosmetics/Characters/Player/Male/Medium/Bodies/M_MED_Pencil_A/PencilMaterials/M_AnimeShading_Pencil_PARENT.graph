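# Parent anime-shading material for the Pencil player cosmetic, using the Cloth shading model on skinned meshes with morph target, clothing and mesh deformer support.
# It works on material attributes, passing the shared character PreFX result through the Pencil anime-shading function and then into the shared PostFX function.
# All the look logic lives in those three functions, with PreFX also feeding PostFX's auxiliary inputs directly.
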Material /BRCosmetics/Characters/Player/Male/Medium/Bodies/M_MED_Pencil_A/PencilMaterials/M_AnimeShading_Pencil_PARENT
graph {
 properties {
  bAutomaticallySetUsageInEditor=false
  bUseMaterialAttributes=true
  bUsedWithClothing=true
  bUsedWithMeshDeformer=true
  bUsedWithMorphTargets=true
  bUsedWithSkeletalMesh=true
  pixelDepthOffsetMode=PDOM_Legacy
  shadingModel=MSM_Cloth
  usedShadingModels=Cloth
 }
 nodes [
  n1 = MaterialFunctionCall() {materialFunction=&BaseFN_CharacterFunction_PreFX}
  MaterialFunctionCall(
   AO=@n1.AO
   Distance Blend=@n1.AO
   Fresnel=@n1.AO
   Gmap=@n1.AO
   Pre Skinned Local Position=@n1.AO
   Material=MaterialFunctionCall!(Material=@n1.Result, CustomUVs, EyeNormal, UV, EyeDebug, DebugCornea) {materialFunction=&/MFC_AnimeShading_Pencil}.Result
  ) {materialFunction=&BaseFN_CharacterFunction_PostFX}
 ]
}
