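# Parent material for the toon-shaded eyes of the Pencil character body, built on material attributes for skeletal meshes.
# MF_EyeUVs supplies eye UVs to the shared character PreFX function.
# The PreFX result runs through MFC_AnimeShading_Pencil for anime-style shading, then into the shared PostFX function.
# All logic lives in those material functions, so the graph itself is only the wiring between them.
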Material /BRCosmetics/Characters/Player/Male/Medium/Bodies/M_MED_Pencil_A/PencilMaterials/M_Eyes_Toon_Pencil_PARENT
graph {
 properties {
  bAutomaticallySetUsageInEditor=false
  bUseMaterialAttributes=true
  bUsedWithMeshDeformer=true
  bUsedWithSkeletalMesh=true
  pixelDepthOffsetMode=PDOM_Legacy
  usedShadingModels=Default Lit
 }
 nodes [
  n1 = MaterialFunctionCall() {materialFunction=/Game/Characters/Shared/FNEye/Materials/MF_EyeUVs}
  n2 = MaterialFunctionCall(TexCoords=@n1.EyeUVs) {materialFunction=&BaseFN_CharacterFunction_PreFX}
  MaterialFunctionCall(
   AO=@n2.AO
   Distance Blend=@n2.AO
   Fresnel=@n2.AO
   Gmap=@n2.AO
   Pre Skinned Local Position=@n2.AO
   Material=MaterialFunctionCall!(Material=@n2.Result, CustomUVs=@n1.EyeUVs, EyeNormal=@n1.EyeUVs, UV=@n1.EyeUVs, EyeDebug=@n1.EyeUVs, DebugCornea=@n1.EyeUVs) {materialFunction=&/MFC_AnimeShading_Pencil}.Result
  ) {materialFunction=&BaseFN_CharacterFunction_PostFX}
 ]
}
